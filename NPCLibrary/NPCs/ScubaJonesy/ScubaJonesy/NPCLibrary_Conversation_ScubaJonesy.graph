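# Conversation database for the Scuba Jonesy NPC in the NPC library, defining what the character says when a player talks to it.
# A single entry point leads to one speech node with a default line plus contextual lines.
# Each contextual line is gated by a player check (time since last match, first conversation, Bars held) and ranked by a curve-table priority.
# The speech then branches through link nodes into the shared NPC library quests and basic services conversations.
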ConversationDatabase /NPCLibrary/NPCs/ScubaJonesy/ScubaJonesy/NPCLibrary_Conversation_ScubaJonesy
entry_points [{tag=Conversation.Entry.ScubaJonesy, destinations [#1]}]
nodes [
 ConversationEntryPointNode_7 {
  id=#1
  class=ConversationEntryPointNode
  outputs [#2]
  properties {entryTag {tagName=Conversation.Entry.ScubaJonesy}}
 }
 Speech_C_4 {
  id=#2
  class=Speech_C
  outputs [#3, #4]
  properties {
   nodeName=""
   generalConfig {
    defaultMessage=It feels so weird to be on land.
    contextualMessages [
     {
      message="Back to check out the fish? I like the blue ones!"
      contextRequirements [{requirement {class=FortControllerRequirement_CheckTimeSinceLastPlayedMatch}}]
      priority {curve {rowName=Tandem.Speech.Priority.WhileSincePlayed}}
     }
     {
      message=You got your diving license? I'm looking for a buddy.
      contextRequirements [{requirement {class=FortControllerRequirement_IsFirstEverConversationWithNPC}}]
      priority {curve {rowName=Tandem.Speech.Priority.FirstEverInteraction}}
     }
     {
      message=You could buy a lot of flippers with those Bars.
      contextRequirements [{requirement {class=FortControllerRequirement_HasBars}}]
      priority {curve {rowName=Tandem.Speech.Priority.HighTotalBars}}
     }
    ]
   }
  }
 }
 ConversationLinkNode_0 {id=#4, class=ConversationLinkNode, properties {remoteEntryTag {tagName=Conversation.Entry.BasicServices}}}
 ConversationLinkNode_1 {id=#3, class=ConversationLinkNode, properties {remoteEntryTag {tagName=Conversation.Entry.NPCLibrary.Quests}}}
]
